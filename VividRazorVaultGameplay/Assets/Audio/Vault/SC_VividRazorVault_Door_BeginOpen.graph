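# Sound cue for the start of the Vivid Razor vault door opening, built from a single wave player with no randomization or mixing.
# It plays in the doors sound class with boosted volume and keeps playing when silent.
# Spatial behavior references a shared door attenuation asset.
# The cue also stores override values for HRTF spatialization, distance low-pass filtering and a distant reverb submix send, without showing the override switched on.
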SoundCue /VividRazorVaultGameplay/Assets/Audio/Vault/SC_VividRazorVault_Door_BeginOpen
root=WavePlayer_0
nodes [{id=WavePlayer_0, properties {soundWaveAssetPtr=/VividRazorVaultGameplay/Assets/Audio/Vault/wav/SFX/VividRazor_Vault_Start_02}}]
properties {
 attenuationSettings=/VividRazorVaultGameplay/Assets/Audio/Vault/settings/ATT_VividRazorVaultGameplay_Door
 maxDistance=4300
 soundClassObject=&Fort_Doors
 virtualizationMode=PlayWhenSilent
 volumeMultiplier=1.5
 attenuationOverrides {
  bAttenuateWithLPF=true
  bEnableLogFrequencyScaling=true
  bEnableSubmixSends=true
  binauralRadius=800
  falloffDistance=2500
  lPFFrequencyAtMax=500
  lPFRadiusMax=2000
  lPFRadiusMin=400
  spatializationAlgorithm=SPATIALIZATION_HRTF
  stereoSpread=800
  attenuationShapeExtents {x=800}
  submixSendSettings [{
    bEnableHPFCutoff=false
    bEnableLPFCutoff=false
    disableManualSendClamp=false
    hPFCutoff=20
    lPFCutoff=20000
    maxSendDistance=7500
    maxSendLevel=0.25
    minSendDistance=0
    minSendLevel=0
    sendLevel=0.2
    sendLevelControlMethod=Linear
    soundSubmix=&Fortnite_Reverb_Distant
    customSendLevelCurve {externalCurve=/Game/Sounds/SubMixes/Reverb/Fortnite_Reverb_Distant_Send}
   }]
 }
}
