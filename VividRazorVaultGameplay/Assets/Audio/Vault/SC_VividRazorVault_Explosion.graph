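# Sound cue for the Vivid Razor vault door explosion.
# A mixer layers the explosion wave, boosted, with a quieter alarm-die wave that a fixed delay starts a fraction of a second later.
# It uses a door-explosion attenuation asset and the large-explosives sound class, and overrides concurrency to cap simultaneous instances.
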SoundCue /VividRazorVaultGameplay/Assets/Audio/Vault/SC_VividRazorVault_Explosion
root=Mixer_0
nodes [
 {id=Delay_0, children [WavePlayer_0], properties {delayMax=0.75, delayMin=0.75}}
 {id=Mixer_0, children [WavePlayer_1,Delay_0], properties {inputVolume [1.25,0.5]}}
 {id=WavePlayer_0, properties {soundWaveAssetPtr=/VividRazorVaultGameplay/Assets/Audio/Vault/wav/SFX/VividRazor_Vault_Door_Expl_AlarmDie}}
 {id=WavePlayer_1, properties {soundWaveAssetPtr=/VividRazorVaultGameplay/Assets/Audio/Vault/wav/SFX/VividRazor_Vault_Door_Expl_01}}
]
properties {
 attenuationSettings=/VividRazorVaultGameplay/Assets/Audio/Vault/settings/ATT_VividRazorVaultGameplay_DoorExpl
 bOverrideConcurrency=true
 maxDistance=12000
 soundClassObject=/Game/Sounds/FortSoundClassesAndModes/Fort_Explosives_Large_PSM
 concurrencyOverrides {maxCount=3}
}
