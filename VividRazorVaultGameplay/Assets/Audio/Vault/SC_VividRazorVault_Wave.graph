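# Sound cue for the Vivid Razor Vault's distant exterior beacon alarm, routed to the gameplay events sound class with boosted volume.
# A single wave player feeds the output.
# Cue-level attenuation overrides the assigned base alarm settings, adding long-range falloff, distance-based low-pass filtering and occlusion.
# Concurrency override allows one instance at a time, stopping the farthest before refusing new ones, with distance-based volume scaling.
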SoundCue /VividRazorVaultGameplay/Assets/Audio/Vault/SC_VividRazorVault_Wave
root=WavePlayer_0
nodes [{id=WavePlayer_0, properties {soundWaveAssetPtr=/VividRazorVaultGameplay/Proto/Sounds/wav/Alarm_Vault_Alert_LP_ExteriorBeacon_Dist}}]
properties {
 attenuationSettings=/Game/Sounds/Asteria/POI/HeistPOIs/Att_Base_Alarms
 bOverrideAttenuation=true
 bOverrideConcurrency=true
 maxDistance=12500
 soundClassObject=&Fort_Events_Gameplay
 volumeMultiplier=2
 attenuationOverrides {
  bAttenuateWithLPF=true
  bEnableOcclusion=true
  falloffDistance=2500
  lPFFrequencyAtMax=900
  lPFFrequencyAtMin=15000
  lPFRadiusMax=8000
  lPFRadiusMin=200
  occlusionLowPassFilterFrequency=2000
  attenuationShapeExtents {x=10000}
 }
 concurrencyOverrides {maxCount=1, resolutionRule=StopFarthestThenPreventNew, voiceStealReleaseTime=1, volumeScaleAttackTime=2, volumeScaleMode=Distance}
}
