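# Sound cue for the destruction of a weakpoint in the VividRazor Vault.
# Picks one of four break variations at random, runs it through a Modulator, then a short randomized delay into the output Mixer.
# Plays boosted in the boss-fight player-critical sound class, using the vault door's attenuation asset.
# Feeds loot, music and vault alarm sidechain volume buses, with a weaker send to loot, likely for ducking.
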SoundCue /VividRazorVaultGameplay/Assets/Audio/Vault/SC_VividRazorVault_Weakpoint_Destroy
root=Mixer_0
nodes [
 {id=Delay_1, children [Modulator_1], properties {delayMax=0.35, delayMin=0.25}}
 {id=Mixer_0, children [Delay_1], properties {inputVolume [1]}}
 {id=Modulator_1, children [Random_0]}
 {id=Random_0, children [WavePlayer_0,WavePlayer_1,WavePlayer_2,WavePlayer_6], properties {weights [1,1,1,1]}}
 {id=WavePlayer_0, properties {soundWaveAssetPtr=/VividRazorVaultGameplay/Assets/Audio/Vault/wav/SFX/VividRazor_Vault_WP_Break_01}}
 {id=WavePlayer_1, properties {soundWaveAssetPtr=/VividRazorVaultGameplay/Assets/Audio/Vault/wav/SFX/VividRazor_Vault_WP_Break_02}}
 {id=WavePlayer_2, properties {soundWaveAssetPtr=/VividRazorVaultGameplay/Assets/Audio/Vault/wav/SFX/VividRazor_Vault_WP_Break_03}}
 {id=WavePlayer_6, properties {soundWaveAssetPtr=/VividRazorVaultGameplay/Assets/Audio/Vault/wav/SFX/VividRazor_Vault_WP_Break_04}}
]
properties {
 attenuationSettings=/VividRazorVaultGameplay/Assets/Audio/Vault/settings/ATT_VividRazorVaultGameplay_Door
 maxDistance=4300
 soundClassObject=/Game/Sounds/FortSoundClassesAndModes/Fort_BossFight_SoundClass/Fort_BossFight_PlayerCritical
 volumeMultiplier=2
 attenuationOverrides {bAttenuate=false, falloffDistance=10000, stereoSpread=500, attenuationShapeExtents {x=2000}}
 busSends [
  {
   audioBus=&AB_Sidechain_FortLoot_Vol
   maxSendLevel=0.3
   minSendLevel=0
   sendLevel=0.3
   soundSourceBus=null
  }
  {audioBus=&AB_Sidechain_FortMusic_Vol, minSendLevel=0, sendLevel=1, soundSourceBus=null}
  {audioBus=/VividRazorVaultGameplay/Assets/Audio/Vault/settings/AB_Sidechain_VaultAlarms_Vol, minSendLevel=0, sendLevel=1, soundSourceBus=null}
 ]
 concurrencyOverrides {maxCount=5, resolutionRule=PreventNew}
}
