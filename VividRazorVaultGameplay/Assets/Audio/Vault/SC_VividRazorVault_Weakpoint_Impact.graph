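# Sound cue for weakpoint impact hits on the Vivid Razor Vault, routed to the boss-fight player-critical sound class.
# It randomly picks one of six equally weighted impact waves and passes it through two chained modulators.
# An "islocal" branch sends local playback through a full-volume mixer, while non-local or unset playback gets half volume and the Door attenuation settings.
# Overridden concurrency caps simultaneous instances, stopping the oldest.
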SoundCue /VividRazorVaultGameplay/Assets/Audio/Vault/SC_VividRazorVault_Weakpoint_Impact
root=Branch_0
nodes [
 {
  id=Attenuation_0
  children [Modulator_0]
  properties {attenuationSettings=/VividRazorVaultGameplay/Assets/Audio/Vault/settings/ATT_VividRazorVaultGameplay_Door}
 }
 {id=Attenuation_1, children [Modulator_0]}
 {id=Branch_0, children [Mixer_0,Mixer_1,Mixer_1], properties {boolParameterName=islocal}}
 {id=Mixer_0, children [Attenuation_1], properties {inputVolume [1]}}
 {id=Mixer_1, children [Attenuation_0], properties {inputVolume [0.5]}}
 {id=Modulator_0, children [Modulator_1], properties {pitchMax=1}}
 {id=Modulator_1, children [Random_0]}
 {id=Random_0, children [WavePlayer_0,WavePlayer_1,WavePlayer_2,WavePlayer_3,WavePlayer_4,WavePlayer_5], properties {weights [1,1,1,1,1,1]}}
 {id=WavePlayer_0, properties {soundWaveAssetPtr=/VividRazorVaultGameplay/Assets/Audio/Vault/wav/SFX/VividRazor_Vault_WP_Impact_01}}
 {id=WavePlayer_1, properties {soundWaveAssetPtr=/VividRazorVaultGameplay/Assets/Audio/Vault/wav/SFX/VividRazor_Vault_WP_Impact_02}}
 {id=WavePlayer_2, properties {soundWaveAssetPtr=/VividRazorVaultGameplay/Assets/Audio/Vault/wav/SFX/VividRazor_Vault_WP_Impact_03}}
 {id=WavePlayer_3, properties {soundWaveAssetPtr=/VividRazorVaultGameplay/Assets/Audio/Vault/wav/SFX/VividRazor_Vault_WP_Impact_04}}
 {id=WavePlayer_4, properties {soundWaveAssetPtr=/VividRazorVaultGameplay/Assets/Audio/Vault/wav/SFX/VividRazor_Vault_WP_Impact_05}}
 {id=WavePlayer_5, properties {soundWaveAssetPtr=/VividRazorVaultGameplay/Assets/Audio/Vault/wav/SFX/VividRazor_Vault_WP_Impact_06}}
]
properties {
 bOverrideConcurrency=true
 maxDistance=2097152
 soundClassObject=/Game/Sounds/FortSoundClassesAndModes/Fort_BossFight_SoundClass/Fort_BossFight_PlayerCritical
 volumeMultiplier=1
 attenuationOverrides {falloffDistance=10000, attenuationShapeExtents {x=0}}
 concurrencyOverrides {maxCount=2, resolutionRule=StopOldest, retriggerTime=0.2}
}
