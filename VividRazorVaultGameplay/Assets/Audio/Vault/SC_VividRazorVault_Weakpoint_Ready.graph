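# Sound cue for the VividRazor vault weakpoint-ready sound, playing weakpoint-exposed effects through the boss-fight player-critical sound class.
# A Random node picks among four equally weighted wave variants, feeding a Modulator that varies playback for each trigger.
# It uses the vault door attenuation asset and caps concurrent instances, blocking new ones at the limit.
# Bus sends feed loot, music and vault-alarm sidechain buses, presumably to duck them under the cue.
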SoundCue /VividRazorVaultGameplay/Assets/Audio/Vault/SC_VividRazorVault_Weakpoint_Ready
root=Modulator_1
nodes [
 {id=Modulator_1, children [Random_0], properties {pitchMin=1}}
 {id=Random_0, children [WavePlayer_0,WavePlayer_1,WavePlayer_2,WavePlayer_3], properties {weights [1,1,1,1]}}
 {id=WavePlayer_0, properties {soundWaveAssetPtr=/VividRazorVaultGameplay/Assets/Audio/Vault/wav/SFX/VividRazor_Vault_WP_Exposed_01}}
 {id=WavePlayer_1, properties {soundWaveAssetPtr=/VividRazorVaultGameplay/Assets/Audio/Vault/wav/SFX/VividRazor_Vault_WP_Exposed_02}}
 {id=WavePlayer_2, properties {soundWaveAssetPtr=/VividRazorVaultGameplay/Assets/Audio/Vault/wav/SFX/VividRazor_Vault_WP_Exposed_03}}
 {id=WavePlayer_3, properties {soundWaveAssetPtr=/VividRazorVaultGameplay/Assets/Audio/Vault/wav/SFX/VividRazor_Vault_WP_Exposed_04}}
]
properties {
 attenuationSettings=/VividRazorVaultGameplay/Assets/Audio/Vault/settings/ATT_VividRazorVaultGameplay_Door
 bOverrideConcurrency=true
 maxDistance=4300
 soundClassObject=/Game/Sounds/FortSoundClassesAndModes/Fort_BossFight_SoundClass/Fort_BossFight_PlayerCritical
 volumeMultiplier=1.5
 attenuationOverrides {bAttenuate=false, falloffDistance=10000, attenuationShapeExtents {x=0}}
 busSends [
  {audioBus=&AB_Sidechain_FortLoot_Vol, minSendLevel=0, sendLevel=1, soundSourceBus=null}
  {audioBus=&AB_Sidechain_FortMusic_Vol, minSendLevel=0, sendLevel=0.3, soundSourceBus=null}
  {audioBus=/VividRazorVaultGameplay/Assets/Audio/Vault/settings/AB_Sidechain_VaultAlarms_Vol, minSendLevel=0, sendLevel=1, soundSourceBus=null}
 ]
 concurrencyOverrides {maxCount=5, resolutionRule=PreventNew}
}
