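# Sound cue for the looping alarm sound of the VividRazor vault.
# A fixed two-second Delay node feeds a single looping wave player, so the alarm starts late and then repeats continuously.
# It plays at reduced volume in the gameplay events sound class, with wide alarm attenuation and a concurrency override capping simultaneous instances at three.
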SoundCue /VividRazorVaultGameplay/Assets/Audio/Vault/SC_VividRazor_Vault_Alarm_Loop_01
root=Delay_5
nodes [
 {id=Delay_5, children [WavePlayer_5], properties {delayMax=2, delayMin=2}}
 {
  id=WavePlayer_5
  properties {bLooping=true, soundWaveAssetPtr=/VividRazorVaultGameplay/Assets/Audio/Vault/wav/SFX/VividRazor_Vault_Alarm_Loop_01}
 }
]
properties {
 attenuationSettings=/VividRazorVaultGameplay/Assets/Audio/Vault/settings/ATT_VividRazor_Alarm_Wide
 maxDistance=10000
 soundClassObject=&Fort_Events_Gameplay
 volumeMultiplier=0.6
 concurrencyOverrides {bVolumeScaleCanRelease=true, maxCount=3}
 concurrencySet [null]
}
