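# Sound cue for a looping alarm in the Vivid Razor vault gameplay content, classed as a gameplay event sound.
# A fixed two-second delay precedes a looping wave player for the alarm loop wave.
# The cue plays at reduced volume with a small-alarm attenuation asset, a limited audible range and a small-alarms concurrency set that caps simultaneous instances.
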SoundCue /VividRazorVaultGameplay/Assets/Audio/Vault/SC_VividRazor_Vault_Alarm_Loop_02
root=Delay_4
nodes [
 {id=Delay_4, children [WavePlayer_1], properties {delayMax=2, delayMin=2}}
 {
  id=WavePlayer_1
  properties {bLooping=true, soundWaveAssetPtr=/VividRazorVaultGameplay/Assets/Audio/Vault/wav/SFX/VividRazor_Vault_Alarm_Loop_02}
 }
]
properties {
 attenuationSettings=/VividRazorVaultGameplay/Assets/Audio/Vault/settings/ATT_VividRazor_Alarm_Small
 maxDistance=2400
 soundClassObject=&Fort_Events_Gameplay
 volumeMultiplier=0.3
 concurrencyOverrides {bVolumeScaleCanRelease=true, maxCount=3}
 concurrencySet [&/CONC_Small_Alarms]
}
