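# Sound cue for a looping vault alarm in the VividRazor gameplay content.
# A fixed two-second delay feeds an equal-weight random pick among three looping alarm wave variants.
# Playback uses small-alarm attenuation and a small-alarm concurrency set, at reduced volume, routed to the gameplay events sound class.
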SoundCue /VividRazorVaultGameplay/Assets/Audio/Vault/SC_VividRazor_Vault_Alarm_Loop_03
root=Delay_4
nodes [
 {id=Delay_4, children [Random_0], properties {delayMax=2, delayMin=2}}
 {id=Random_0, children [WavePlayer_0,WavePlayer_2,WavePlayer_3], properties {weights [1,1,1]}}
 {
  id=WavePlayer_0
  properties {bLooping=true, soundWaveAssetPtr=/VividRazorVaultGameplay/Assets/Audio/Vault/wav/SFX/VividRazor_Vault_Alarm_Loop_03}
 }
 {
  id=WavePlayer_2
  properties {bLooping=true, soundWaveAssetPtr=/VividRazorVaultGameplay/Assets/Audio/Vault/wav/SFX/VividRazor_Vault_Alarm_Loop_05}
 }
 {
  id=WavePlayer_3
  properties {bLooping=true, soundWaveAssetPtr=/VividRazorVaultGameplay/Assets/Audio/Vault/wav/SFX/VividRazor_Vault_Alarm_Loop_04}
 }
]
properties {
 attenuationSettings=/VividRazorVaultGameplay/Assets/Audio/Vault/settings/ATT_VividRazor_Alarm_Small
 maxDistance=2400
 soundClassObject=&Fort_Events_Gameplay
 volumeMultiplier=0.5
 concurrencyOverrides {bVolumeScaleCanRelease=true, maxCount=3}
 concurrencySet [&/CONC_Small_Alarms]
}
