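# Small vault alarm cue for VividRazor vault gameplay, played spatially with small-alarm attenuation and concurrency on the gameplay events sound class.
# A mixer layers three branches: a delayed one-shot lockdown alarm pitched down to half, and a delayed repeating random pick of three quieter alarm waves.
# The third branch randomly picks one of two looping vault waves after a delay, one with varied pitch, under an envelope that lowers volume over time.
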SoundCue /VividRazorVaultGameplay/Assets/Audio/Vault/SC_VividRazor_Vault_Alarm_Loop_04
root=Mixer_0
nodes [
 {id=Delay_2, children [Modulator_2], properties {delayMax=2.5, delayMin=2.5}}
 {id=Delay_3, children [WavePlayer_3], properties {delayMax=2, delayMin=2}}
 {id=Delay_4, children [Random_1], properties {delayMax=2, delayMin=2}}
 {
  id=Enveloper_0
  children [Delay_4]
  properties {pitchCurve {editorCurveData {keys [{time=0}]}}, volumeCurve {editorCurveData {keys [<at 0 1>,<at 15 1>,<at 25 0.7>]}}}
 }
 {id=Looping_5, children [Delay_2]}
 {id=Mixer_0, children [Modulator_1,Looping_5,Enveloper_0], properties {inputVolume [1,1,1]}}
 {id=Modulator_0, children [WavePlayer_1], properties {pitchMax=1, pitchMin=0.8}}
 {id=Modulator_1, children [Delay_3], properties {pitchMax=0.5, pitchMin=0.5}}
 {id=Modulator_2, children [Random_0], properties {pitchMax=1, pitchMin=1, volumeMax=0.68, volumeMin=0.68}}
 {id=Random_0, children [WavePlayer_3,WavePlayer_0,WavePlayer_2], properties {weights [1,1,1]}}
 {id=Random_1, children [Modulator_0,WavePlayer_4], properties {weights [1,1]}}
 {id=WavePlayer_0, properties {soundWaveAssetPtr=/Game/Sounds/Asteria/POI/HeistPOIs/wav/Alarm_Lockdown_LP_01}}
 {
  id=WavePlayer_1
  properties {bLooping=true, soundWaveAssetPtr=/VividRazorVaultGameplay/Assets/Audio/Vault/wav/SFX/VividRazor_Vault_Alarm_Loop_06}
 }
 {id=WavePlayer_2, properties {soundWaveAssetPtr=/VividRazorVaultGameplay/Assets/Audio/Vault/wav/SFX/VividRazor_Vault_Alarm_Loop_03}}
 {id=WavePlayer_3, properties {soundWaveAssetPtr=/Game/Sounds/Asteria/POI/HeistPOIs/wav/Alarm_Lockdown_LP_06}}
 {
  id=WavePlayer_4
  properties {bLooping=true, soundWaveAssetPtr=/VividRazorVaultGameplay/Assets/Audio/Vault/wav/SFX/VividRazor_Vault_Alarm_Loop_07}
 }
]
properties {
 attenuationSettings=/VividRazorVaultGameplay/Assets/Audio/Vault/settings/ATT_VividRazor_Alarm_Small
 maxDistance=2400
 soundClassObject=&Fort_Events_Gameplay
 volumeMultiplier=0.5
 concurrencyOverrides {bVolumeScaleCanRelease=true, maxCount=3}
 concurrencySet [&/CONC_Small_Alarms]
}
