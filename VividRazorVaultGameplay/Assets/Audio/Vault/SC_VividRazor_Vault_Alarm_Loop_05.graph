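# Sound cue for the looping Vivid Razor vault alarm, classed as a gameplay event sound with wide alarm attenuation.
# A looping wave player (Alarm_Loop_09 wave, despite the cue's 05 suffix) feeds a fixed two-second delay, an enveloper and a single-input mixer.
# The envelope fades volume in quickly, holds it, then eases it to a lower level over a longer stretch.
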SoundCue /VividRazorVaultGameplay/Assets/Audio/Vault/SC_VividRazor_Vault_Alarm_Loop_05
root=Mixer_2
nodes [
 {id=Delay_5, children [WavePlayer_0], properties {delayMax=2, delayMin=2}}
 {
  id=Enveloper_0
  children [Delay_5]
  properties {pitchCurve {editorCurveData {keys [{time=0}]}}, volumeCurve {editorCurveData {keys [<at 0 0>,<at 0.5 1>,<at 15 1>,<at 25 0.7>]}}}
 }
 {id=Mixer_2, children [Enveloper_0], properties {inputVolume [1]}}
 {
  id=WavePlayer_0
  properties {bLooping=true, soundWaveAssetPtr=/VividRazorVaultGameplay/Assets/Audio/Vault/wav/SFX/VividRazor_Vault_Alarm_Loop_09}
 }
]
properties {
 attenuationSettings=/VividRazorVaultGameplay/Assets/Audio/Vault/settings/ATT_VividRazor_Alarm_Wide
 maxDistance=10000
 soundClassObject=&Fort_Events_Gameplay
 volumeMultiplier=1
 concurrencyOverrides {bVolumeScaleCanRelease=true, maxCount=3}
 concurrencySet [null]
}
