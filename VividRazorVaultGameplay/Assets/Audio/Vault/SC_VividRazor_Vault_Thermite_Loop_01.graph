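# Sound cue for the looping thermite burn sound in the VividRazor vault gameplay, likely at a vault door given its attenuation asset.
# A looping wave player feeds an enveloper whose volume curve fades the sound in from silence over the first second.
# It plays at reduced volume on the gameplay events sound class, with door-loop attenuation and a concurrency override limiting simultaneous instances.
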SoundCue /VividRazorVaultGameplay/Assets/Audio/Vault/SC_VividRazor_Vault_Thermite_Loop_01
root=Enveloper_4
nodes [
 {
  id=Enveloper_4
  children [WavePlayer_6]
  properties {pitchCurve {editorCurveData {keys [{time=0}]}}, volumeCurve {editorCurveData {keys [<at 0 0>,<at 1 1>]}}}
 }
 {
  id=WavePlayer_6
  properties {bLooping=true, soundWaveAssetPtr=/VividRazorVaultGameplay/Assets/Audio/Vault/wav/SFX/VividRazor_Vault_Thermite_Loop_01}
 }
]
properties {
 attenuationSettings=/VividRazorVaultGameplay/Assets/Audio/Vault/settings/ATT_VividRazor_Thermite_Door_Loop
 maxDistance=2500
 soundClassObject=&Fort_Events_Gameplay
 volumeMultiplier=0.5
 concurrencyOverrides {bVolumeScaleCanRelease=true, maxCount=3}
 concurrencySet [null]
}
